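# Physics asset for the skirt of the NitroFlow Royal medium female body, defining capsule bodies for its dynamic bones.
# Six skirt chains and two cloth chains hang from master bodies joined to the pelvis, with per-segment swing and twist limits.
# Free-angle constraints also cross-link some neighboring skirt chains at middle segments.
# Pelvis, thigh and calf capsules model the lower body, alongside a large offset root capsule.
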PhysicsAsset /BRCosmetics/Characters/Player/Female/Medium/Bodies/F_MED_NitroFlow_Royal/Meshes/F_MED_NitroFlow_Royal_Skirt_Physics
bodies {
 calf_l [{shapeType=Capsule, length=48, radius=9, shapeName=calf_l_capsule, center=<-33.7105 -3.5 0>, rotation=<rot 0 90 90>}]
 calf_r [{shapeType=Capsule, length=48, radius=9, shapeName=calf_l_capsule, center=<33.7105 3.5 0>, rotation=<rot 0 91.9782 90>}]
 dyn_cloth_l_01 [{shapeType=Capsule, length=12, radius=4, shapeName=dyn_cloth_l_01_capsule, center=<7 -1 2>, rotation=<rot 90 0 0>}]
 dyn_cloth_l_02 [{shapeType=Capsule, length=10, radius=6.3, shapeName=dyn_cloth_l_02_capsule, center=<7 -1 2>, rotation=<rot 90 0 0>}]
 dyn_cloth_l_03 [{shapeType=Capsule, length=10, radius=7, shapeName=dyn_cloth_l_03_capsule, center=<7 0 2>, rotation=<rot 90 0 0>}]
 dyn_cloth_l_04 [{shapeType=Capsule, length=10, radius=4, shapeName=dyn_cloth_l_04_capsule, center=<7 0 2>, rotation=<rot 90 0 0>}]
 dyn_cloth_l_mstr [{shapeType=Capsule, length=2, radius=2, shapeName=dyn_cloth_l_mstr_capsule, rotation=<rot 90 0 0>}]
 dyn_cloth_r_01 [{shapeType=Capsule, length=12, radius=4, shapeName=dyn_cloth_r_01_capsule, center=<7 1 2>, rotation=<rot 90 0 0>}]
 dyn_cloth_r_02 [{shapeType=Capsule, length=10, radius=6.3, shapeName=dyn_cloth_r_02_capsule, center=<7 1 2>, rotation=<rot 90 0 0>}]
 dyn_cloth_r_03 [{shapeType=Capsule, length=10, radius=7, shapeName=dyn_cloth_r_03_capsule, center=<7 0 2>, rotation=<rot 90 0 0>}]
 dyn_cloth_r_04 [{shapeType=Capsule, length=10, radius=4, shapeName=dyn_cloth_r_04_capsule, center=<7 0 2>, rotation=<rot 90 0 0>}]
 dyn_cloth_r_mstr [{shapeType=Capsule, length=2, radius=2, shapeName=dyn_cloth_r_mstr_capsule, rotation=<rot 90 0 0>}]
 dyn_skirt_01_l_01 [{shapeType=Capsule, length=8, radius=9, shapeName=dyn_skirt_01_l_01_capsule, center=<10 0 9>, rotation=<rot 0 0 90>}]
 dyn_skirt_01_l_02 [{shapeType=Capsule, length=8, radius=10, shapeName=dyn_skirt_01_l_02_capsule, center=<10 0 9>, rotation=<rot 0 0 90>}]
 dyn_skirt_01_l_03 [{shapeType=Capsule, length=8, radius=10, shapeName=dyn_skirt_01_l_03_capsule, center=<12 0 5>, rotation=<rot 0 0 97.6245>}]
 dyn_skirt_01_l_04 [{shapeType=Capsule, length=6.25, radius=10, shapeName=dyn_skirt_01_l_04_capsule, center=<10 0 5>, rotation=<rot 0 0 90>}]
 dyn_skirt_01_l_05 [{shapeType=Capsule, length=6.25, radius=10, shapeName=dyn_skirt_01_l_05_capsule, center=<5 0 5>, rotation=<rot 0 0 90>}]
 dyn_skirt_01_l_mstr [{shapeType=Capsule, length=1, radius=1, shapeName=dyn_skirt_01_l_mstr_capsule}]
 dyn_skirt_01_r_01 [{shapeType=Capsule, length=8, radius=9, shapeName=dyn_skirt_01_r_01_capsule, center=<10 0 9>, rotation=<rot 0 0 90>}]
 dyn_skirt_01_r_02 [{shapeType=Capsule, length=8, radius=10, shapeName=dyn_skirt_01_r_02_capsule, center=<10 0 9>, rotation=<rot 0 0 90>}]
 dyn_skirt_01_r_03 [{shapeType=Capsule, length=8, radius=9, shapeName=dyn_skirt_01_r_03_capsule, center=<12 0 5>, rotation=<rot 0 0 90>}]
 dyn_skirt_01_r_04 [{shapeType=Capsule, length=8, radius=9, shapeName=dyn_skirt_01_r_04_capsule, center=<9 0 5>, rotation=<rot 0 0 90>}]
 dyn_skirt_01_r_05 [{shapeType=Capsule, length=8, radius=9, shapeName=dyn_skirt_01_r_05_capsule, center=<9 0 5>, rotation=<rot 0 0 90>}]
 dyn_skirt_01_r_mstr [{shapeType=Capsule, length=1, radius=1, shapeName=dyn_skirt_01_r_mstr_capsule}]
 dyn_skirt_02_l_01 [{shapeType=Capsule, length=8, radius=10, shapeName=dyn_skirt_02_l_01_capsule, center=<10 0 9>, rotation=<rot 0 0 90>}]
 dyn_skirt_02_l_02 [{shapeType=Capsule, length=14.5, radius=11, shapeName=dyn_skirt_02_l_02_capsule, center=<10 0 9>, rotation=<rot 0 0 90>}]
 dyn_skirt_02_l_03 [{shapeType=Capsule, length=17.25, radius=11.5, shapeName=dyn_skirt_02_l_03_capsule, center=<10 0 5>, rotation=<rot 0 0 90>}]
 dyn_skirt_02_l_04 [{shapeType=Capsule, length=20.75, radius=9, shapeName=dyn_skirt_02_l_04_capsule, center=<10 0 5>, rotation=<rot -0.0001 0 91.258>}]
 dyn_skirt_02_l_mstr [{shapeType=Capsule, length=1, radius=1, shapeName=dyn_skirt_02_l_mstr_capsule}]
 dyn_skirt_02_r_01 [{shapeType=Capsule, length=8, radius=10, shapeName=dyn_skirt_02_r_01_capsule, center=<10 0 9>, rotation=<rot 0 0 90>}]
 dyn_skirt_02_r_02 [{shapeType=Capsule, length=14.5, radius=11, shapeName=dyn_skirt_02_r_02_capsule, center=<10 0 9>, rotation=<rot 0 0 97.0453>}]
 dyn_skirt_02_r_03 [{shapeType=Capsule, length=22.75, radius=11.5, shapeName=dyn_skirt_02_r_03_capsule, center=<10 0 5>, rotation=<rot 0 0 97.4557>}]
 dyn_skirt_02_r_04 [{shapeType=Capsule, length=29, radius=9, shapeName=dyn_skirt_02_r_04_capsule, center=<10 0 5>, rotation=<rot -0.0245 20.6654 93.8004>}]
 dyn_skirt_02_r_mstr [{shapeType=Capsule, length=1, radius=1, shapeName=dyn_skirt_02_r_mstr_capsule}]
 dyn_skirt_bL_01 [{shapeType=Capsule, length=8, radius=10, shapeName=dyn_skirt_02_l_01_capsule, center=<10 0 9>, rotation=<rot 0 0 90>}]
 dyn_skirt_bL_02 [{shapeType=Capsule, length=14.5, radius=11, shapeName=dyn_skirt_02_l_02_capsule, center=<10 0 9>, rotation=<rot 0 0 90>}]
 dyn_skirt_bL_03 [{shapeType=Capsule, length=17.25, radius=11.5, shapeName=dyn_skirt_02_l_03_capsule, center=<10 0 5>, rotation=<rot 0 0 90>}]
 dyn_skirt_bL_04 [{shapeType=Capsule, length=23.75, radius=11, shapeName=dyn_skirt_02_l_04_capsule, center=<10 0 5>, rotation=<rot -0.0001 0 91.258>}]
 dyn_skirt_bL_mstr [{shapeType=Capsule, length=1, radius=1, shapeName=dyn_skirt_02_r_mstr_capsule}]
 dyn_skirt_bR_01 [{shapeType=Capsule, length=8, radius=10, shapeName=dyn_skirt_02_l_01_capsule, center=<10 0 9>, rotation=<rot 0 0 90>}]
 dyn_skirt_bR_02 [{shapeType=Capsule, length=14.5, radius=11, shapeName=dyn_skirt_02_l_02_capsule, center=<10 0 9>, rotation=<rot 0 0 90>}]
 dyn_skirt_bR_03 [{shapeType=Capsule, length=17.25, radius=11.5, shapeName=dyn_skirt_02_l_03_capsule, center=<10 0 5>, rotation=<rot 0 0 90>}]
 dyn_skirt_bR_04 [{shapeType=Capsule, length=23.75, radius=11, shapeName=dyn_skirt_02_l_04_capsule, center=<10 0 5>, rotation=<rot -0.0001 0 91.258>}]
 dyn_skirt_bR_mstr [{shapeType=Capsule, length=1, radius=1, shapeName=dyn_skirt_02_r_mstr_capsule}]
 pelvis [{shapeType=Capsule, length=14.9001, radius=12, rotation=<rot 90 159.444 152.1102>}]
 root [{shapeType=Capsule, length=63.75, radius=106.25, shapeName=root_capsule, center=<0 0 -88.9913>, rotation=<rot 90 165.9638 75.9638>}]
 thigh_l [{shapeType=Capsule, length=39, radius=7.5, shapeName=thigh_l_capsule, center=<-21.3981 0 0>, rotation=<rot 0 90 90>}]
 thigh_r [{shapeType=Capsule, length=39, radius=7, shapeName=thigh_l_capsule, center=<21.3981 0 0>, rotation=<rot 0 90 90>}]
}
constraints [
 {
  bone1Name=dyn_skirt_01_l_mstr
  bone2Name=pelvis
  swing1LimitDegrees=45
  swing1Motion=Free
  swing2LimitDegrees=45
  swing2Motion=Free
  twistLimitDegrees=45
  twistMotion=Free
 }
 {bone1Name=dyn_skirt_01_l_01, bone2Name=dyn_skirt_01_l_mstr, swing1LimitDegrees=5, swing2LimitDegrees=30, twistLimitDegrees=2}
 {bone1Name=dyn_skirt_01_l_02, bone2Name=dyn_skirt_01_l_01, swing1LimitDegrees=5, swing2LimitDegrees=30, twistLimitDegrees=5}
 {bone1Name=dyn_skirt_01_l_03, bone2Name=dyn_skirt_01_l_02, swing1LimitDegrees=5, swing2LimitDegrees=40, twistLimitDegrees=10}
 {bone1Name=dyn_skirt_01_l_04, bone2Name=dyn_skirt_01_l_03, swing1LimitDegrees=5, swing2LimitDegrees=25, twistLimitDegrees=2}
 {bone1Name=dyn_skirt_01_l_05, bone2Name=dyn_skirt_01_l_04, swing1LimitDegrees=15, swing2LimitDegrees=30, twistLimitDegrees=2}
 {
  bone1Name=dyn_skirt_01_r_mstr
  bone2Name=pelvis
  swing1LimitDegrees=45
  swing1Motion=Free
  swing2LimitDegrees=45
  swing2Motion=Free
  twistLimitDegrees=45
  twistMotion=Free
 }
 {bone1Name=dyn_skirt_01_r_01, bone2Name=dyn_skirt_01_r_mstr, swing1LimitDegrees=5, swing2LimitDegrees=30, twistLimitDegrees=2}
 {bone1Name=dyn_skirt_01_r_02, bone2Name=dyn_skirt_01_r_01, swing1LimitDegrees=5, swing2LimitDegrees=30, twistLimitDegrees=2}
 {bone1Name=dyn_skirt_01_r_03, bone2Name=dyn_skirt_01_r_02, swing1LimitDegrees=5, swing2LimitDegrees=40, twistLimitDegrees=2}
 {bone1Name=dyn_skirt_01_r_04, bone2Name=dyn_skirt_01_r_03, swing1LimitDegrees=5, swing2LimitDegrees=34, twistLimitDegrees=2}
 {bone1Name=dyn_skirt_01_r_05, bone2Name=dyn_skirt_01_r_04, swing1LimitDegrees=15, swing2LimitDegrees=30, twistLimitDegrees=2}
 {
  bone1Name=dyn_skirt_02_l_mstr
  bone2Name=pelvis
  swing1LimitDegrees=45
  swing1Motion=Free
  swing2LimitDegrees=45
  swing2Motion=Free
  twistLimitDegrees=45
  twistMotion=Free
 }
 {bone1Name=dyn_skirt_02_l_01, bone2Name=dyn_skirt_02_l_mstr, swing1LimitDegrees=20, swing2LimitDegrees=20, twistLimitDegrees=2}
 {bone1Name=dyn_skirt_02_l_02, bone2Name=dyn_skirt_02_l_01, swing1LimitDegrees=20, swing2LimitDegrees=20, twistLimitDegrees=2}
 {bone1Name=dyn_skirt_02_l_03, bone2Name=dyn_skirt_02_l_02, swing1LimitDegrees=20, swing2LimitDegrees=40, twistLimitDegrees=2}
 {bone1Name=dyn_skirt_02_l_04, bone2Name=dyn_skirt_02_l_03, swing1LimitDegrees=25, swing2LimitDegrees=45, twistLimitDegrees=2}
 {
  bone1Name=dyn_skirt_02_r_mstr
  bone2Name=pelvis
  swing1LimitDegrees=45
  swing1Motion=Free
  swing2LimitDegrees=45
  swing2Motion=Free
  twistLimitDegrees=45
  twistMotion=Free
 }
 {bone1Name=dyn_skirt_02_r_01, bone2Name=dyn_skirt_02_r_mstr, swing1LimitDegrees=20, swing2LimitDegrees=20, twistLimitDegrees=2}
 {bone1Name=dyn_skirt_02_r_02, bone2Name=dyn_skirt_02_r_01, swing1LimitDegrees=20, swing2LimitDegrees=25, twistLimitDegrees=3}
 {bone1Name=dyn_skirt_02_r_03, bone2Name=dyn_skirt_02_r_02, swing1LimitDegrees=25, swing2LimitDegrees=35, twistLimitDegrees=3}
 {bone1Name=dyn_skirt_02_r_04, bone2Name=dyn_skirt_02_r_03, swing1LimitDegrees=35, swing2LimitDegrees=45, twistLimitDegrees=3}
 {
  bone1Name=thigh_l
  bone2Name=pelvis
  swing1LimitDegrees=45
  swing1Motion=Free
  swing2LimitDegrees=45
  swing2Motion=Free
  twistLimitDegrees=45
  twistMotion=Free
 }
 {
  bone1Name=calf_l
  bone2Name=thigh_l
  swing1LimitDegrees=45
  swing1Motion=Free
  swing2LimitDegrees=45
  swing2Motion=Free
  twistLimitDegrees=45
  twistMotion=Free
 }
 {
  bone1Name=thigh_r
  bone2Name=pelvis
  swing1LimitDegrees=45
  swing1Motion=Free
  swing2LimitDegrees=45
  swing2Motion=Free
  twistLimitDegrees=45
  twistMotion=Free
 }
 {
  bone1Name=calf_r
  bone2Name=thigh_r
  swing1LimitDegrees=45
  swing1Motion=Free
  swing2LimitDegrees=45
  swing2Motion=Free
  twistLimitDegrees=45
  twistMotion=Free
 }
 {
  bone1Name=pelvis
  bone2Name=root
  swing1LimitDegrees=45
  swing1Motion=Free
  swing2LimitDegrees=45
  swing2Motion=Free
  twistLimitDegrees=45
  twistMotion=Free
 }
 {
  bone1Name=dyn_skirt_01_r_02
  bone2Name=dyn_skirt_02_r_02
  swing1LimitDegrees=45
  swing1Motion=Free
  swing2LimitDegrees=45
  swing2Motion=Free
  twistLimitDegrees=45
  twistMotion=Free
 }
 {
  bone1Name=dyn_skirt_01_l_02
  bone2Name=dyn_skirt_02_l_02
  swing1LimitDegrees=45
  swing1Motion=Free
  swing2LimitDegrees=45
  swing2Motion=Free
  twistLimitDegrees=45
  twistMotion=Free
 }
 {bone1Name=dyn_skirt_bL_01, bone2Name=dyn_skirt_bL_mstr, swing1LimitDegrees=25, swing2LimitDegrees=25, twistLimitDegrees=2}
 {bone1Name=dyn_skirt_bL_02, bone2Name=dyn_skirt_bL_01, swing1LimitDegrees=30, swing2LimitDegrees=30, twistLimitDegrees=2}
 {bone1Name=dyn_skirt_bL_03, bone2Name=dyn_skirt_bL_02, swing1LimitDegrees=35, swing2LimitDegrees=40, twistLimitDegrees=2}
 {bone1Name=dyn_skirt_bL_04, bone2Name=dyn_skirt_bL_03, swing1LimitDegrees=45, swing2LimitDegrees=45, twistLimitDegrees=2}
 {bone1Name=dyn_skirt_bR_01, bone2Name=dyn_skirt_bR_mstr, swing1LimitDegrees=25, swing2LimitDegrees=25, twistLimitDegrees=2}
 {bone1Name=dyn_skirt_bR_02, bone2Name=dyn_skirt_bR_01, swing1LimitDegrees=30, swing2LimitDegrees=30, twistLimitDegrees=2}
 {bone1Name=dyn_skirt_bR_03, bone2Name=dyn_skirt_bR_02, swing1LimitDegrees=35, swing2LimitDegrees=35, twistLimitDegrees=2}
 {bone1Name=dyn_skirt_bR_04, bone2Name=dyn_skirt_bR_03, swing1LimitDegrees=45, swing2LimitDegrees=45, twistLimitDegrees=2}
 {
  bone1Name=dyn_skirt_bL_mstr
  bone2Name=pelvis
  swing1LimitDegrees=45
  swing1Motion=Free
  swing2LimitDegrees=45
  swing2Motion=Free
  twistLimitDegrees=45
  twistMotion=Free
 }
 {
  bone1Name=dyn_skirt_bR_mstr
  bone2Name=pelvis
  swing1LimitDegrees=45
  swing1Motion=Free
  swing2LimitDegrees=45
  swing2Motion=Free
  twistLimitDegrees=45
  twistMotion=Free
 }
 {
  bone1Name=dyn_skirt_02_r_03
  bone2Name=dyn_skirt_bR_03
  swing1LimitDegrees=45
  swing1Motion=Free
  swing2LimitDegrees=45
  swing2Motion=Free
  twistLimitDegrees=45
  twistMotion=Free
 }
 {bone1Name=dyn_cloth_r_04, bone2Name=dyn_cloth_r_03, swing1LimitDegrees=10, swing2LimitDegrees=25, twistLimitDegrees=10}
 {
  bone1Name=dyn_cloth_r_mstr
  bone2Name=pelvis
  swing1LimitDegrees=45
  swing1Motion=Free
  swing2LimitDegrees=45
  swing2Motion=Free
  twistLimitDegrees=45
  twistMotion=Free
 }
 {bone1Name=dyn_cloth_r_01, bone2Name=dyn_cloth_r_mstr, swing1LimitDegrees=20, swing2LimitDegrees=35, twistLimitDegrees=10}
 {bone1Name=dyn_cloth_r_02, bone2Name=dyn_cloth_r_01, swing1LimitDegrees=15, swing2LimitDegrees=25, twistLimitDegrees=10}
 {bone1Name=dyn_cloth_r_03, bone2Name=dyn_cloth_r_02, swing1LimitDegrees=10, swing2LimitDegrees=25, twistLimitDegrees=10}
 {bone1Name=dyn_cloth_l_04, bone2Name=dyn_cloth_l_03, swing1LimitDegrees=10, swing2LimitDegrees=25, twistLimitDegrees=10}
 {
  bone1Name=dyn_cloth_l_mstr
  bone2Name=pelvis
  swing1LimitDegrees=45
  swing1Motion=Free
  swing2LimitDegrees=45
  swing2Motion=Free
  twistLimitDegrees=45
  twistMotion=Free
 }
 {bone1Name=dyn_cloth_l_01, bone2Name=dyn_cloth_l_mstr, swing1LimitDegrees=15, swing2LimitDegrees=25, twistLimitDegrees=10}
 {bone1Name=dyn_cloth_l_02, bone2Name=dyn_cloth_l_01, swing1LimitDegrees=15, swing2LimitDegrees=25, twistLimitDegrees=10}
 {bone1Name=dyn_cloth_l_03, bone2Name=dyn_cloth_l_02, swing1LimitDegrees=10, swing2LimitDegrees=25, twistLimitDegrees=10}
 {
  bone1Name=dyn_skirt_bR_03
  bone2Name=dyn_skirt_bL_03
  swing1LimitDegrees=45
  swing1Motion=Free
  swing2LimitDegrees=45
  swing2Motion=Free
  twistLimitDegrees=45
  twistMotion=Free
 }
 {
  bone1Name=dyn_skirt_bL_03
  bone2Name=dyn_skirt_bR_03
  swing1LimitDegrees=45
  swing1Motion=Free
  swing2LimitDegrees=45
  swing2Motion=Free
  twistLimitDegrees=45
  twistMotion=Free
 }
 {
  bone1Name=dyn_skirt_02_l_03
  bone2Name=dyn_skirt_01_l_03
  swing1LimitDegrees=45
  swing1Motion=Free
  swing2LimitDegrees=45
  swing2Motion=Free
  twistLimitDegrees=45
  twistMotion=Free
 }
]
