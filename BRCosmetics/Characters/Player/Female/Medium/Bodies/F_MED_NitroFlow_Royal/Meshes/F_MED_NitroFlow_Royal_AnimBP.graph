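# Anim Blueprint for the F_MED_NitroFlow_Royal female body cosmetic, adding secondary motion to skirt, cuffs, ropes and cloth.
# It copies the parent pose through deform and outfit Control Rigs, then simulates with state-specific RigidBody variants and AnimDynamics chains.
# Without rigid bodies, trail controllers and crouch offsets substitute; in vehicles, a scaled, constrained skirt pose takes over.
# The EventGraph derives gravity override, movement alphas and vehicle state each update.
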AnimBlueprint /BRCosmetics/Characters/Player/Female/Medium/Bodies/F_MED_NitroFlow_Royal/Meshes/F_MED_NitroFlow_Royal_AnimBP
target_skeleton=&/F_MED_NitroFlow_Royal_Skeleton
graphs [
 AnimGraph {nodes [
   n1 = Get in vehicle scale alpha() -> (in vehicle scale alpha:double)
   n2 = "Transform (Modify) Bone - Bone: dyn_skirt_01_l_01"(ComponentPose=Local To Component(), Alpha=@n1, Translation="0,0,0", Rotation="0,0,0", Scale="0.3,0.8,0.8") {scaleMode=BMM_Replace, alphaBoolBlend {blendOption=Linear}, boneToModify {boneName=dyn_skirt_01_l_01}, scale=<0.3 0.8 0.8>}
   n3 = "Transform (Modify) Bone - Bone: dyn_skirt_02_l_01"(ComponentPose=@n2, Alpha=@n1, Translation="0,0,0", Rotation="0,0,0", Scale="0.3,0.8,0.8") {scaleMode=BMM_Replace, alphaBoolBlend {blendOption=Linear}, boneToModify {boneName=dyn_skirt_02_l_01}, scale=<0.3 0.8 0.8>}
   n4 = "Transform (Modify) Bone - Bone: dyn_skirt_bL_01"(ComponentPose=@n3, Alpha=@n1, Translation="0,0,0", Rotation="0,0,0", Scale="0.3,0.8,0.8") {scaleMode=BMM_Replace, alphaBoolBlend {blendOption=Linear}, boneToModify {boneName=dyn_skirt_bL_01}, scale=<0.3 0.8 0.8>}
   n5 = "Transform (Modify) Bone - Bone: dyn_skirt_01_r_01"(ComponentPose=@n4, Alpha=@n1, Translation="0,0,0", Rotation="0,0,0", Scale="0.3,0.8,0.8") {scaleMode=BMM_Replace, alphaBoolBlend {blendOption=Linear}, boneToModify {boneName=dyn_skirt_01_r_01}, scale=<0.3 0.8 0.8>}
   n6 = "Transform (Modify) Bone - Bone: dyn_skirt_02_r_01"(ComponentPose=@n5, Alpha=@n1, Translation="0,0,0", Rotation="0,0,0", Scale="0.3,0.8,0.8") {scaleMode=BMM_Replace, alphaBoolBlend {blendOption=Linear}, boneToModify {boneName=dyn_skirt_02_r_01}, scale=<0.3 0.8 0.8>}
   n7 = "Transform (Modify) Bone - Bone: dyn_skirt_bR_01"(ComponentPose=@n6, Alpha=@n1, Translation="0,0,0", Rotation="0,0,0", Scale="0.3,0.8,0.8") {scaleMode=BMM_Replace, alphaBoolBlend {blendOption=Linear}, boneToModify {boneName=dyn_skirt_bR_01}, scale=<0.3 0.8 0.8>}
   n8 = "Transform (Modify) Bone - Bone: dyn_skirt_b_01"(ComponentPose=@n7, Alpha=@n1, Translation="0,0,0", Rotation="0,0,0", Scale="0.3,0.8,0.8") {scaleMode=BMM_Replace, alphaBoolBlend {blendOption=Linear}, boneToModify {boneName=dyn_skirt_b_01}, scale=<0.3 0.8 0.8>}
   n9 = "Transform (Modify) Bone - Bone: dyn_cloth_l_01"(ComponentPose=@n8, Alpha=1, Translation="0,0,0", Rotation="0,0,0", Scale="0.5,0.8,0.8") {scaleMode=BMM_Replace, alphaBoolBlend {blendOption=Linear}, boneToModify {boneName=dyn_cloth_l_01}, scale=<0.5 0.8 0.8>}
   n10 = "Transform (Modify) Bone - Bone: dyn_cloth_r_01"(ComponentPose=@n9, Alpha=1, Translation="0,0,0", Rotation="0,0,0", Scale="0.5,0.8,0.8") {scaleMode=BMM_Replace, alphaBoolBlend {blendOption=Linear}, boneToModify {boneName=dyn_cloth_r_01}, scale=<0.5 0.8 0.8>}
   Component To Local(ComponentPose=@n10)
   "Transform (Modify) Bone - Bone: dyn_cuff_l_mstr"(Alpha=1, Translation="0,0,0", Rotation="28.890,26.728,-20.407", Scale="1,1,1") {
    rotationMode=BMM_Additive
    rotationSpace=BCS_BoneSpace
    alphaBoolBlend {blendOption=Linear}
    boneToModify {boneName=dyn_cuff_l_mstr}
    rotation=<rot 28.89 26.728 -20.407>
   }
   n11 = Get Gravity Override()
   n12 = RigidBody!(ComponentPose=Local To Component(LocalPose=Use cached pose 'beforeRban'()), Alpha, OverrideWorldGravity:Vector=@n11, ExternalForce) -> (Pose) {
    bOverrideWorldGravity=true
    overridePhysicsAsset=&/F_MED_NitroFlow_Royal_Skirt_TactSprint_Physics
    simulationSpace=BaseBoneSpace
    alphaBoolBlend {blendOption=Linear}
    baseBoneRef {boneName=pelvis}
    componentAppliedLinearAccClamp=<3000 3000 3000>
    componentLinearAccScale=<0.1 0.1 0.4>
    componentLinearVelScale=<0.1 0.1 0.4>
    simSpaceSettings {maxAngularAcceleration=3000, maxAngularVelocity=0, maxLinearAcceleration=1000, maxLinearVelocity=2000, velocityScaleZ=0.2, worldAlpha=0.2}
   }
   n13 = Component To Local(ComponentPose=@n12)
   n14 = RigidBody(
    ComponentPose=Local To Component(LocalPose=Use cached pose 'beforeRban'())
    Alpha=Get NormalRbanAlpha() -> (NormalRbanAlpha:double)
    OverrideWorldGravity=@n11
   ) {
    bOverrideWorldGravity=true
    overridePhysicsAsset=&/F_MED_NitroFlow_Royal_Skirt_Physics
    simulationSpace=BaseBoneSpace
    alphaBoolBlend {blendOption=Linear}
    baseBoneRef {boneName=pelvis}
    componentAppliedLinearAccClamp=<3000 3000 3000>
    componentLinearAccScale=<0.4 0.4 0.7>
    componentLinearVelScale=<0.4 0.4 0.7>
    simSpaceSettings {maxAngularAcceleration=3000, maxAngularVelocity=0, maxLinearAcceleration=1000, maxLinearVelocity=2000, velocityScaleZ=0.2, worldAlpha=1}
   }
   n15 = Component To Local(ComponentPose=@n14)
   n16 = RigidBody(
    ComponentPose=Local To Component(LocalPose=Blend Poses by bool(BlendPose_0=@n13, BlendPose_1=@n15, BlendTime_0=1, BlendTime_1=1) {blendType=Linear, transitionType=Inertialization})
    OverrideWorldGravity=@n11
   ) {
    bOverrideWorldGravity=true
    overridePhysicsAsset=&/F_MED_NitroFlow_Royal_Skirt_frontend_Physics
    simulationSpace=BaseBoneSpace
    alphaBoolBlend {blendOption=Linear}
    baseBoneRef {boneName=pelvis}
    componentAppliedLinearAccClamp=<3000 3000 3000>
    componentLinearAccScale=<0.6 0.6 0.7>
    componentLinearVelScale=<0.6 0.6 0.7>
    simSpaceSettings {maxAngularAcceleration=3000, maxAngularVelocity=0, maxLinearAcceleration=1000, maxLinearVelocity=2000, worldAlpha=1}
   }
   n17 = Component To Local(ComponentPose=@n16)
   n18 = RigidBody(ComponentPose=Local To Component(LocalPose=Use cached pose 'beforeRban'()), OverrideWorldGravity="0,0,0") {
    bOverrideWorldGravity=true
    overridePhysicsAsset=&/F_MED_NitroFlow_Royal_Skirt_Sliding_Physics
    simulationSpace=BaseBoneSpace
    alphaBoolBlend {blendOption=Linear}
    baseBoneRef {boneName=pelvis}
    componentAppliedLinearAccClamp=<3000 3000 3000>
    componentLinearAccScale=<0.7 0.7 2>
    componentLinearVelScale=<0.7 0.7 2>
    simSpaceSettings {maxAngularVelocity=0, maxLinearAcceleration=700, maxLinearVelocity=500, velocityScaleZ=5, worldAlpha=0.7}
   }
   n19 = Component To Local(ComponentPose=@n18)
   n20 = Get DynamicAlpha() -> (DynamicAlpha:double)
   n21 = Anim Dynamics - Bone: dyn_cuff_l_mstr(
    ComponentPose=Local To Component(LocalPose=Blend Poses by bool(BlendPose_0=@n19, BlendPose_1=@n17) {blendType=Linear})
    Alpha=@n20
    ExternalForce="0,0,0"
   ) {
    angularDampingOverride=0.9
    angularSpringConstant=50
    bOverrideAngularDamping=true
    simulationSpace=RootRelative
    alphaBoolBlend {blendOption=Linear}
    boundBone {boneName=dyn_cuff_l_mstr}
    componentAppliedLinearAccClamp=<10000 10000 10000>
    componentLinearAccScale=<1 1 1>
    componentLinearVelScale=<1 1 1>
    physicsBodyDefinitions [{
      boundBone {boneName=dyn_cuff_l_mstr}
      boxExtents=<25 25 25>
      constraintSetup {angularLimitsMax=<10 0 35>, angularLimitsMin=<-10 0 -35>, angularTarget=<0 0 0>}
      localJointOffset=<1 0 0>
     }]
    relativeSpaceBone {boneName=lowerarm_l}
   }
   n22 = Anim Dynamics - Bone: dyn_cuff_l_01(ComponentPose=@n21, Alpha=@n20, ExternalForce="0,0,0") {
    angularDampingOverride=1
    angularSpringConstant=400
    bOverrideAngularDamping=true
    simulationSpace=RootRelative
    alphaBoolBlend {blendOption=Linear}
    boundBone {boneName=dyn_cuff_l_01}
    componentAppliedLinearAccClamp=<1000 1000 1000>
    componentLinearAccScale=<0.5 0.5 0.5>
    componentLinearVelScale=<0.5 0.5 0.5>
    physicsBodyDefinitions [{
      boundBone {boneName=dyn_cuff_l_01}
      boxExtents=<60 60 60>
      constraintSetup {angularLimitsMax=<3 3 50>, angularLimitsMin=<-3 -3 -50>, angularTarget=<0 0 0>}
      localJointOffset=<10 0 0>
     }]
    relativeSpaceBone {boneName=lowerarm_l}
   }
   n23 = Anim Dynamics - Bone: dyn_cuff_l_02(ComponentPose=@n22, Alpha=1, ExternalForce="0,0,0") {
    angularDampingOverride=0.999
    angularSpringConstant=150
    simulationSpace=RootRelative
    alphaBoolBlend {blendOption=Linear}
    boundBone {boneName=dyn_cuff_l_02}
    componentAppliedLinearAccClamp=<3000 3000 3000>
    componentLinearAccScale=<1 1 1>
    componentLinearVelScale=<1 1 1>
    physicsBodyDefinitions [{
      boundBone {boneName=dyn_cuff_l_02}
      boxExtents=<8 8 8>
      constraintSetup {angularLimitsMax=<0 15 60>, angularLimitsMin=<0 -15 -60>, angularTarget=<0 0 0>}
      localJointOffset=<0.2 0 0>
     }]
    relativeSpaceBone {boneName=lowerarm_l}
   }
   n24 = Anim Dynamics - Bone: dyn_cuff_r_mstr(ComponentPose=@n23, Alpha=@n20, ExternalForce="0,0,0") {
    angularDampingOverride=0.9
    angularSpringConstant=50
    bAngularSpring=true
    bOverrideAngularDamping=true
    simulationSpace=RootRelative
    alphaBoolBlend {blendOption=Linear}
    boundBone {boneName=dyn_cuff_r_mstr}
    componentAppliedLinearAccClamp=<10000 10000 10000>
    componentLinearAccScale=<1 1 1>
    componentLinearVelScale=<1 1 1>
    physicsBodyDefinitions [{
      boundBone {boneName=dyn_cuff_r_mstr}
      boxExtents=<25 25 25>
      constraintSetup {angularLimitsMax=<10 0 35>, angularLimitsMin=<-10 0 -35>, angularTarget=<0 0 0>}
      localJointOffset=<1 0 0>
     }]
    relativeSpaceBone {boneName=lowerarm_l}
   }
   n25 = Anim Dynamics - Bone: dyn_cuff_r_01(ComponentPose=@n24, Alpha=1, ExternalForce="0,0,0") {
    angularDampingOverride=1
    angularSpringConstant=600
    bAngularSpring=true
    bOverrideAngularDamping=true
    simulationSpace=RootRelative
    alphaBoolBlend {blendOption=Linear}
    boundBone {boneName=dyn_cuff_r_01}
    componentAppliedLinearAccClamp=<1000 1000 1000>
    componentLinearAccScale=<0.5 0.5 0.5>
    componentLinearVelScale=<0.5 0.5 0.5>
    physicsBodyDefinitions [{
      boundBone {boneName=dyn_cuff_r_01}
      boxExtents=<20 20 20>
      constraintSetup {angularLimitsMax=<5 3 40>, angularLimitsMin=<-5 -3 -40>, angularTarget=<0 0 0>}
      localJointOffset=<8 0 0>
     }]
    relativeSpaceBone {boneName=lowerarm_l}
   }
   n26 = Anim Dynamics - Bone: dyn_cuff_r_02(ComponentPose=@n25, Alpha=1, ExternalForce="0,0,0") {
    angularDampingOverride=1
    angularSpringConstant=200
    bAngularSpring=true
    simulationSpace=RootRelative
    alphaBoolBlend {blendOption=Linear}
    boundBone {boneName=dyn_cuff_r_02}
    componentAppliedLinearAccClamp=<3000 3000 3000>
    componentLinearAccScale=<0.5 0.5 0.5>
    componentLinearVelScale=<0.5 0.5 0.5>
    physicsBodyDefinitions [{
      boundBone {boneName=dyn_cuff_r_02}
      boxExtents=<10 10 12>
      constraintSetup {angularLimitsMax=<0 15 65>, angularLimitsMin=<0 -15 -65>, angularTarget=<0 0 0>}
      localJointOffset=<0.2 0 0>
     }]
    relativeSpaceBone {boneName=lowerarm_l}
   }
   n27 = Anim Dynamics - Bone: dyn_fold(ComponentPose=@n26, Alpha=1, ExternalForce="0,0,0") {
    angularDampingOverride=0.9999
    angularSpringConstant=200
    bAngularSpring=true
    bOverrideAngularDamping=true
    simulationSpace=BoneRelative
    alphaBoolBlend {blendOption=Linear}
    boundBone {boneName=dyn_fold}
    componentLinearAccScale=<1 1 3>
    componentLinearVelScale=<1 1 3>
    physicsBodyDefinitions [{
      boundBone {boneName=dyn_fold}
      boxExtents=<35 35 35>
      constraintSetup {angularLimitsMax=<0 20 0>, angularLimitsMin=<0 -5 0>, angularTarget=<0 0 0>}
      localJointOffset=<0 0 -5>
     }]
    relativeSpaceBone {boneName=pelvis}
   }
   n28 = "Constraint - dyn_roll_mstr (PP)"!(ComponentPose=@n27, Alpha=1, ConstraintWeights_0=0.5, ConstraintWeights_1:float=0.5) -> (Pose) {
    alphaBoolBlend {blendOption=Linear}
    boneToModify {boneName=dyn_roll_mstr}
    constraintSetup [{targetBone {boneName=dyn_skirt_bL_01}}, {targetBone {boneName=dyn_skirt_bR_01}}]
    constraintWeights [1,1]
   }
   n29 = Anim Dynamics - Bone: dyn_roll_01(ComponentPose=@n28, Alpha=1, ExternalForce="0,0,0") {
    angularDampingOverride=0.9999
    angularSpringConstant=50
    bAngularSpring=true
    bOverrideAngularDamping=true
    simulationSpace=RootRelative
    alphaBoolBlend {blendOption=Linear}
    boundBone {boneName=dyn_roll_01}
    componentAppliedLinearAccClamp=<10000 10000 10000>
    componentLinearAccScale=<0.5 0.5 0.5>
    componentLinearVelScale=<0.5 0.5 0.5>
    physicsBodyDefinitions [{
      boundBone {boneName=dyn_roll_01}
      boxExtents=<100 100 100>
      constraintSetup {angularLimitsMax=<0 0 20>, angularLimitsMin=<0 -30 -20>, angularTarget=<0 0 0>}
      localJointOffset=<10 0 0>
     }]
    relativeSpaceBone {boneName=pelvis}
   }
   n30 = Anim Dynamics - Bone: dyn_roll_02(ComponentPose=@n29, Alpha=1, ExternalForce="0,0,0") {
    angularDampingOverride=0.99
    angularSpringConstant=200
    bOverrideAngularDamping=true
    simulationSpace=RootRelative
    alphaBoolBlend {blendOption=Linear}
    boundBone {boneName=dyn_roll_02}
    componentAppliedLinearAccClamp=<10000 10000 10000>
    componentLinearAccScale=<1 1 1>
    componentLinearVelScale=<1 1 1>
    physicsBodyDefinitions [{
      boundBone {boneName=dyn_roll_02}
      boxExtents=<80 80 80>
      constraintSetup {angularLimitsMax=<0 0 20>, angularLimitsMin=<0 -40 -20>, angularTarget=<0 0 0>}
      localJointOffset=<10 0 0>
     }]
    relativeSpaceBone {boneName=pelvis}
   }
   n31 = "Transform (Modify) Bone - Bone: dyn_roll_mstr"(ComponentPose=@n30, Alpha=1, Translation="0,0,0.5", Rotation="0,0,5", Scale="1,1,1") {
    rotationMode=BMM_Additive
    translationMode=BMM_Additive
    alphaBoolBlend {blendOption=Linear}
    boneToModify {boneName=dyn_roll_mstr}
    rotation {roll=5}
    translation {z=0.5}
   }
   n32 = "Constraint - dyn_rope_l_mstr (RR)"!(ComponentPose=@n31, Alpha=1, ConstraintWeights_0=0.5, ConstraintWeights_1:float=0.5) -> (Pose) {
    alphaBoolBlend {blendOption=Linear}
    boneToModify {boneName=dyn_rope_l_mstr}
    constraintSetup [{transformType=Rotation, targetBone {boneName=dyn_cloth_l_01}}, {transformType=Rotation, targetBone {boneName=dyn_cloth_r_01}}]
    constraintWeights [1,1]
   }
   n33 = Anim Dynamics - Bone: dyn_rope_l_01(ComponentPose=@n32, Alpha=1, ExternalForce="0,0,0") {
    angularDampingOverride=0.8
    bOverrideAngularDamping=true
    simulationSpace=RootRelative
    alphaBoolBlend {blendOption=Linear}
    boundBone {boneName=dyn_rope_l_01}
    componentAppliedLinearAccClamp=<10000 10000 10000>
    componentLinearAccScale=<0.3 0.3 0.3>
    componentLinearVelScale=<0.3 0.3 0.3>
    physicsBodyDefinitions [{
      boundBone {boneName=dyn_rope_l_01}
      boxExtents=<20 20 20>
      constraintSetup {angularLimitsMax=<0 0 7>, angularLimitsMin=<0 -15 -7>, angularTarget=<0 0 0>}
      localJointOffset=<1.5 0 0>
     }]
   }
   n34 = Anim Dynamics - Bone: dyn_rope_l_02(ComponentPose=@n33, Alpha=1, ExternalForce="0,0,0") {
    angularDampingOverride=0.9
    bOverrideAngularDamping=true
    simulationSpace=RootRelative
    alphaBoolBlend {blendOption=Linear}
    boundBone {boneName=dyn_rope_l_02}
    componentAppliedLinearAccClamp=<10000 10000 10000>
    componentLinearAccScale=<0.3 0.3 0.3>
    componentLinearVelScale=<0.3 0.3 0.3>
    physicsBodyDefinitions [{
      boundBone {boneName=dyn_rope_l_02}
      boxExtents=<20 20 20>
      constraintSetup {angularLimitsMax=<0 0 15>, angularLimitsMin=<0 -15 -15>, angularTarget=<0 0 0>}
      localJointOffset=<1.5 0 0>
     }]
   }
   n35 = Anim Dynamics - Bone: dyn_rope_l_03(ComponentPose=@n34, Alpha=1, ExternalForce="0,0,0") {
    angularDampingOverride=0.9
    bOverrideAngularDamping=true
    simulationSpace=RootRelative
    alphaBoolBlend {blendOption=Linear}
    boundBone {boneName=dyn_rope_l_03}
    componentAppliedLinearAccClamp=<10000 10000 10000>
    componentLinearAccScale=<0.3 0.3 0.3>
    componentLinearVelScale=<0.3 0.3 0.3>
    physicsBodyDefinitions [{
      boundBone {boneName=dyn_rope_l_03}
      boxExtents=<20 20 20>
      constraintSetup {angularLimitsMax=<0 10 15>, angularLimitsMin=<0 -15 -15>, angularTarget=<0 0 0>}
      localJointOffset=<1.2 0 0>
     }]
   }
   n36 = "Constraint - dyn_rope_r_mstr (RR)"!(ComponentPose=@n35, Alpha=1, ConstraintWeights_0=0, ConstraintWeights_1:float=1) -> (Pose) {
    alphaBoolBlend {blendOption=Linear}
    boneToModify {boneName=dyn_rope_r_mstr}
    constraintSetup [{transformType=Rotation, targetBone {boneName=dyn_cloth_l_mstr}}, {transformType=Rotation, targetBone {boneName=thigh_r}}]
    constraintWeights [1,1]
   }
   n37 = Anim Dynamics - Bone: dyn_rope_r_01(ComponentPose=@n36, Alpha=1, ExternalForce="0,0,0") {
    simulationSpace=RootRelative
    alphaBoolBlend {blendOption=Linear}
    boundBone {boneName=dyn_rope_r_01}
    physicsBodyDefinitions [{
      boundBone {boneName=dyn_rope_r_01}
      boxExtents=<30 30 30>
      constraintSetup {angularLimitsMax=<0 0 7>, angularLimitsMin=<0 -15 -7>, angularTarget=<0 0 0>}
      localJointOffset=<1 0 0>
     }]
   }
   n38 = Anim Dynamics - Bone: dyn_rope_r_02(ComponentPose=@n37, Alpha=1, ExternalForce="0,0,0") {
    simulationSpace=RootRelative
    alphaBoolBlend {blendOption=Linear}
    boundBone {boneName=dyn_rope_r_02}
    physicsBodyDefinitions [{
      boundBone {boneName=dyn_rope_r_02}
      boxExtents=<20 20 20>
      constraintSetup {angularLimitsMax=<0 0 15>, angularLimitsMin=<0 -15 -15>, angularTarget=<0 0 0>}
      localJointOffset=<1 0 0>
     }]
   }
   n39 = "Transform (Modify) Bone - Bone: dyn_rope_r_mstr"(ComponentPose=@n38, Alpha=1, Translation="0,0,0", Rotation="0,0,-10", Scale="1,1,1") {
    rotationMode=BMM_Additive
    translationMode=BMM_Additive
    alphaBoolBlend {blendOption=Linear}
    boneToModify {boneName=dyn_rope_r_mstr}
    rotation {roll=-10}
   }
   n40 = "Constraint - dyn_cloth_l_01 (P)"(ComponentPose=@n39, Alpha=1, ConstraintWeights_0=1) {
    alphaBoolBlend {blendOption=Linear}
    boneToModify {boneName=dyn_cloth_l_01}
    constraintSetup [{targetBone {boneName=pelvis}}]
    constraintWeights [1]
   }
   n41 = "Constraint - dyn_cloth_r_01 (P)"(ComponentPose=@n40, Alpha=1, ConstraintWeights_0=1) {
    alphaBoolBlend {blendOption=Linear}
    boneToModify {boneName=dyn_cloth_r_01}
    constraintSetup [{targetBone {boneName=pelvis}}]
    constraintWeights [1]
   }
   n42 = Component To Local(ComponentPose=@n41)
   n43 = Get CrouchScale() -> (CrouchScale:double)
   n44 = "Transform (Modify) Bone - Bone: dyn_skirt_bL_mstr"(
    ComponentPose=Local To Component(LocalPose=Use cached pose 'beforeRban'())
    Alpha=@n43
    Translation="0,0,0"
    Rotation="0,0,45"
    Scale="1,1,1"
   ) {
    rotationMode=BMM_Additive
    translationMode=BMM_Additive
    alphaBoolBlend {blendOption=Linear}
    alphaScaleBiasClamp {bInterpResult=true, interpSpeedDecreasing=1}
    boneToModify {boneName=dyn_skirt_bL_mstr}
    rotation {roll=45}
   }
   n45 = "Transform (Modify) Bone - Bone: dyn_skirt_bR_mstr"(ComponentPose=@n44, Alpha=@n43, Translation="0,0,0", Rotation="0,0,45", Scale="1,1,1") {
    rotationMode=BMM_Additive
    translationMode=BMM_Additive
    alphaBoolBlend {blendOption=Linear}
    alphaScaleBiasClamp {bInterpResult=true, interpSpeedDecreasing=1}
    boneToModify {boneName=dyn_skirt_bR_mstr}
    rotation {roll=45}
   }
   n46 = "Transform (Modify) Bone - Bone: dyn_skirt_02_l_mstr"(ComponentPose=@n45, Alpha=@n43, Translation="0,0,0", Rotation="25,0,30", Scale="1,1,1") {
    rotationMode=BMM_Additive
    translationMode=BMM_Additive
    alphaBoolBlend {blendOption=Linear}
    alphaScaleBiasClamp {bInterpResult=true, interpSpeedDecreasing=1}
    boneToModify {boneName=dyn_skirt_02_l_mstr}
    rotation {pitch=25, roll=30}
   }
   n47 = "Transform (Modify) Bone - Bone: dyn_skirt_02_r_mstr"(ComponentPose=@n46, Alpha=@n43, Translation="0,0,0", Rotation="-25,0,30", Scale="1,1,1") {
    rotationMode=BMM_Additive
    translationMode=BMM_Additive
    alphaBoolBlend {blendOption=Linear}
    alphaScaleBiasClamp {bInterpResult=true, interpSpeedDecreasing=1}
    boneToModify {boneName=dyn_skirt_02_r_mstr}
    rotation {pitch=-25, roll=30}
   }
   n48 = Get BwdMovement() -> (BwdMovement:double)
   n49 = Trail controller - Bone: dyn_skirt_bL_04(ComponentPose=@n47, Alpha=@n48) {
    bUsePlanarLimit=true
    chainLength=4
    debugLifeTime=0.1
    maxDeltaTime=0.0333
    alphaScaleBiasClamp {bInterpResult=true}
    rotationLimits [{limitMax=<180 180 180>}, {limitMax=<180 180 180>}, {limitMax=<180 180 180>}, {limitMax=<180 180 180>}]
    rotationOffsets [<0 0 0>, <0 0 0>, <0 0 0>, <0 0 0>]
    trailBone {boneName=dyn_skirt_bL_04}
    trailRelaxationSpeed {editorCurveData {keys [<at 0 15 RCIM_Linear>,<at 0.5289 13.1838 RCIM_Linear>,<at 1 6 RCIM_Linear>]}}
   }
   n50 = Trail controller - Bone: dyn_skirt_bR_04(ComponentPose=@n49, Alpha=@n48) {
    bUsePlanarLimit=true
    chainLength=4
    debugLifeTime=0.1
    maxDeltaTime=0.0333
    alphaScaleBiasClamp {bInterpResult=true}
    rotationLimits [{limitMax=<180 180 180>}, {limitMax=<180 180 180>}, {limitMax=<180 180 180>}, {limitMax=<180 180 180>}]
    rotationOffsets [<0 0 0>, <0 0 0>, <0 0 0>, <0 0 0>]
    trailBone {boneName=dyn_skirt_bR_04}
    trailRelaxationSpeed {editorCurveData {keys [<at 0 15 RCIM_Linear>,<at 0.5289 13.1838 RCIM_Linear>,<at 1 6 RCIM_Linear>]}}
   }
   n51 = Trail controller - Bone: dyn_skirt_02_r_04(ComponentPose=@n50, Alpha=@n48) {
    bUsePlanarLimit=true
    chainLength=3
    debugLifeTime=0.1
    maxDeltaTime=0.0333
    alphaScaleBiasClamp {bInterpResult=true}
    rotationLimits [{limitMax=<180 180 180>}, {limitMax=<180 180 180>}, {limitMax=<180 180 180>}]
    rotationOffsets [<0 0 0>, <0 0 0>, <0 0 0>]
    trailBone {boneName=dyn_skirt_02_r_04}
    trailRelaxationSpeed {editorCurveData {keys [<at 0 15 RCIM_Linear>,<at 0.5289 13.1838 RCIM_Linear>,<at 1 6 RCIM_Linear>]}}
   }
   n52 = Trail controller - Bone: dyn_skirt_01_r_04(ComponentPose=@n51, Alpha=@n48) {
    bUsePlanarLimit=true
    debugLifeTime=0.1
    maxDeltaTime=0.0333
    alphaScaleBiasClamp {bInterpResult=true}
    rotationLimits [{limitMax=<180 180 180>}, {limitMax=<180 180 180>}]
    rotationOffsets [<0 0 0>, <0 0 0>]
    trailBone {boneName=dyn_skirt_01_r_04}
    trailRelaxationSpeed {editorCurveData {keys [<at 0 15 RCIM_Linear>,<at 0.5289 13.1838 RCIM_Linear>,<at 1 6 RCIM_Linear>]}}
   }
   n53 = Trail controller - Bone: dyn_skirt_02_l_04(ComponentPose=@n52, Alpha=@n48) {
    bUsePlanarLimit=true
    chainLength=3
    debugLifeTime=0.1
    maxDeltaTime=0.0333
    alphaScaleBiasClamp {bInterpResult=true}
    rotationLimits [{limitMax=<180 180 180>}, {limitMax=<180 180 180>}, {limitMax=<180 180 180>}]
    rotationOffsets [<0 0 0>, <0 0 0>, <0 0 0>]
    trailBone {boneName=dyn_skirt_02_l_04}
    trailRelaxationSpeed {editorCurveData {keys [<at 0 15 RCIM_Linear>,<at 0.5289 13.1838 RCIM_Linear>,<at 1 6 RCIM_Linear>]}}
   }
   n54 = Trail controller - Bone: dyn_skirt_01_l_03(ComponentPose=@n53, Alpha=@n48) {
    bUsePlanarLimit=true
    debugLifeTime=0.1
    maxDeltaTime=0.0333
    alphaScaleBiasClamp {bInterpResult=true}
    rotationLimits [{limitMax=<180 180 180>}, {limitMax=<180 180 180>}]
    rotationOffsets [<0 0 0>, <0 0 0>]
    trailBone {boneName=dyn_skirt_01_l_03}
    trailRelaxationSpeed {editorCurveData {keys [<at 0 15 RCIM_Linear>,<at 0.5289 13.1838 RCIM_Linear>,<at 1 6 RCIM_Linear>]}}
   }
   n55 = "Constraint - dyn_roll_mstr (PP)"(ComponentPose=@n54, Alpha=1, ConstraintWeights_0=0.5, ConstraintWeights_1=0.5) {
    alphaBoolBlend {blendOption=Linear}
    boneToModify {boneName=dyn_roll_mstr}
    constraintSetup [{targetBone {boneName=dyn_skirt_bL_01}}, {targetBone {boneName=dyn_skirt_bR_01}}]
    constraintWeights [1,1]
   }
   n56 = Component To Local(ComponentPose=@n55)
   n57 = Get bRigidBodyEnabled()
   n58 = "Transform (Modify) Bone - Bone: dyn_skirt_01_l_01"(
    ComponentPose=Local To Component(LocalPose=Copy Pose From Mesh() {bUseAttachedParent=true})
    Alpha=1
    Translation="0,0,0"
    Rotation="0,0,0"
    Scale="0.5,0.8,0.8"
   ) {scaleMode=BMM_Replace, alphaBoolBlend {blendOption=Linear}, boneToModify {boneName=dyn_skirt_01_l_01}, scale=<0.5 0.8 0.8>}
   n59 = "Transform (Modify) Bone - Bone: dyn_skirt_02_l_01"(ComponentPose=@n58, Alpha=1, Translation="0,0,0", Rotation="0,0,0", Scale="0.5,0.8,0.8") {scaleMode=BMM_Replace, alphaBoolBlend {blendOption=Linear}, boneToModify {boneName=dyn_skirt_02_l_01}, scale=<0.5 0.8 0.8>}
   n60 = "Transform (Modify) Bone - Bone: dyn_skirt_bL_01"(ComponentPose=@n59, Alpha=1, Translation="0,0,0", Rotation="0,0,0", Scale="0.5,0.8,0.8") {scaleMode=BMM_Replace, alphaBoolBlend {blendOption=Linear}, boneToModify {boneName=dyn_skirt_bL_01}, scale=<0.5 0.8 0.8>}
   n61 = "Transform (Modify) Bone - Bone: dyn_skirt_01_r_01"(ComponentPose=@n60, Alpha=1, Translation="0,0,0", Rotation="0,0,0", Scale="0.5,0.8,0.8") {scaleMode=BMM_Replace, alphaBoolBlend {blendOption=Linear}, boneToModify {boneName=dyn_skirt_01_r_01}, scale=<0.5 0.8 0.8>}
   n62 = "Transform (Modify) Bone - Bone: dyn_skirt_02_r_01"(ComponentPose=@n61, Alpha=1, Translation="0,0,0", Rotation="0,0,0", Scale="0.5,0.8,0.8") {scaleMode=BMM_Replace, alphaBoolBlend {blendOption=Linear}, boneToModify {boneName=dyn_skirt_02_r_01}, scale=<0.5 0.8 0.8>}
   n63 = "Transform (Modify) Bone - Bone: dyn_skirt_bR_01"(ComponentPose=@n62, Alpha=1, Translation="0,0,0", Rotation="0,0,0", Scale="0.5,0.8,0.8") {scaleMode=BMM_Replace, alphaBoolBlend {blendOption=Linear}, boneToModify {boneName=dyn_skirt_bR_01}, scale=<0.5 0.8 0.8>}
   n64 = "Transform (Modify) Bone - Bone: dyn_skirt_b_01"(ComponentPose=@n63, Alpha=1, Translation="0,0,0", Rotation="0,0,0", Scale="0.5,0.8,0.8") {scaleMode=BMM_Replace, alphaBoolBlend {blendOption=Linear}, boneToModify {boneName=dyn_skirt_b_01}, scale=<0.5 0.8 0.8>}
   n65 = "Transform (Modify) Bone - Bone: dyn_cloth_l_01"(ComponentPose=@n64, Alpha=1, Translation="0,0,0", Rotation="0,0,0", Scale="0.5,0.8,0.8") {scaleMode=BMM_Replace, alphaBoolBlend {blendOption=Linear}, boneToModify {boneName=dyn_cloth_l_01}, scale=<0.5 0.8 0.8>}
   n66 = "Transform (Modify) Bone - Bone: dyn_cloth_r_01"(ComponentPose=@n65, Alpha=1, Translation="0,0,0", Rotation="0,0,0", Scale="0.5,0.8,0.8") {scaleMode=BMM_Replace, alphaBoolBlend {blendOption=Linear}, boneToModify {boneName=dyn_cloth_r_01}, scale=<0.5 0.8 0.8>}
   n67 = "Constraint - dyn_skirt_01_l_01 (R)"(ComponentPose=@n66, Alpha=1, ConstraintWeights_0=1) {
    alphaBoolBlend {blendOption=Linear}
    boneToModify {boneName=dyn_skirt_01_l_01}
    constraintSetup [{transformType=Rotation, targetBone {boneName=thigh_l}}]
    constraintWeights [1]
   }
   n68 = "Transform (Modify) Bone - Bone: dyn_skirt_01_l_01"(ComponentPose=@n67, Alpha=1, Translation="-2.909,2.483,2.620", Rotation="-6.289,-0.388,0.106", Scale="1,1,1") {
    rotationMode=BMM_Additive
    rotationSpace=BCS_BoneSpace
    translationMode=BMM_Additive
    translationSpace=BCS_BoneSpace
    alphaBoolBlend {blendOption=Linear}
    boneToModify {boneName=dyn_skirt_01_l_01}
    rotation=<rot -6.289 -0.3879 0.106>
    translation=<-2.909 2.483 2.62>
   }
   n69 = "Constraint - dyn_skirt_02_l_01 (RR)"!(ComponentPose=@n68, Alpha=1, ConstraintWeights_0=0.5, ConstraintWeights_1:float=0.5) -> (Pose) {
    alphaBoolBlend {blendOption=Linear}
    boneToModify {boneName=dyn_skirt_02_l_01}
    constraintSetup [{transformType=Rotation, targetBone {boneName=thigh_l}}, {transformType=Rotation, targetBone {boneName=pelvis}}]
    constraintWeights [1,1]
   }
   n70 = "Transform (Modify) Bone - Bone: dyn_skirt_02_l_01"(ComponentPose=@n69, Alpha=1, Translation="0,0,0", Rotation="-14.943,4.021,-20.113", Scale="1,1,1") {
    rotationMode=BMM_Additive
    rotationSpace=BCS_BoneSpace
    translationMode=BMM_Additive
    translationSpace=BCS_BoneSpace
    alphaBoolBlend {blendOption=Linear}
    boneToModify {boneName=dyn_skirt_02_l_01}
    rotation=<rot -14.943 4.0211 -20.113>
   }
   n71 = "Constraint - dyn_skirt_bL_01 (RR)"!(ComponentPose=@n70, Alpha=1, ConstraintWeights_0=0.2, ConstraintWeights_1:float=0.8) -> (Pose) {
    alphaBoolBlend {blendOption=Linear}
    boneToModify {boneName=dyn_skirt_bL_01}
    constraintSetup [{transformType=Rotation, targetBone {boneName=thigh_l}}, {transformType=Rotation, targetBone {boneName=pelvis}}]
    constraintWeights [1,1]
   }
   n72 = "Transform (Modify) Bone - Bone: dyn_skirt_bL_01"(ComponentPose=@n71, Alpha=1, Translation="9.576,-3.842,8.623", Rotation="-38.947,10.050,-17.828", Scale="1,1,1") {
    rotationMode=BMM_Additive
    rotationSpace=BCS_BoneSpace
    translationMode=BMM_Additive
    translationSpace=BCS_BoneSpace
    alphaBoolBlend {blendOption=Linear}
    boneToModify {boneName=dyn_skirt_bL_01}
    rotation=<rot -38.947 10.05 -17.828>
    translation=<9.576 -3.842 8.623>
   }
   n73 = "Constraint - dyn_skirt_01_r_01 (R)"(ComponentPose=@n72, Alpha=1, ConstraintWeights_0=1) {
    alphaBoolBlend {blendOption=Linear}
    boneToModify {boneName=dyn_skirt_01_r_01}
    constraintSetup [{transformType=Rotation, targetBone {boneName=thigh_l}}]
    constraintWeights [1]
   }
   n74 = "Transform (Modify) Bone - Bone: dyn_skirt_01_r_01"(ComponentPose=@n73, Alpha=1, Translation="-2.909,-2.483,2.62", Rotation="30.389,14.643,51.153", Scale="1,1,1") {
    rotationMode=BMM_Additive
    rotationSpace=BCS_BoneSpace
    translationMode=BMM_Additive
    translationSpace=BCS_BoneSpace
    alphaBoolBlend {blendOption=Linear}
    boneToModify {boneName=dyn_skirt_01_r_01}
    rotation=<rot 30.389 14.643 51.153>
    translation=<-2.909 -2.483 2.62>
   }
   n75 = "Constraint - dyn_skirt_02_r_01 (RR)"!(ComponentPose=@n74, Alpha=1, ConstraintWeights_0=0.5, ConstraintWeights_1:float=0.5) -> (Pose) {
    alphaBoolBlend {blendOption=Linear}
    boneToModify {boneName=dyn_skirt_02_r_01}
    constraintSetup [{transformType=Rotation, targetBone {boneName=thigh_l}}, {transformType=Rotation, targetBone {boneName=pelvis}}]
    constraintWeights [1,1]
   }
   n76 = "Transform (Modify) Bone - Bone: dyn_skirt_02_r_01"(ComponentPose=@n75, Alpha=1, Translation="5.329,1.073,1.417", Rotation="5.509,-4.879,25.060", Scale="1,1,1") {
    rotationMode=BMM_Additive
    rotationSpace=BCS_BoneSpace
    translationMode=BMM_Additive
    translationSpace=BCS_BoneSpace
    alphaBoolBlend {blendOption=Linear}
    boneToModify {boneName=dyn_skirt_02_r_01}
    rotation=<rot 5.509 -4.879 25.06>
    translation=<5.329 1.073 1.417>
   }
   n77 = "Constraint - dyn_skirt_bR_01 (RR)"!(ComponentPose=@n76, Alpha=1, ConstraintWeights_0=0.2, ConstraintWeights_1:float=0.8) -> (Pose) {
    alphaBoolBlend {blendOption=Linear}
    boneToModify {boneName=dyn_skirt_bR_01}
    constraintSetup [{transformType=Rotation, targetBone {boneName=thigh_l}}, {transformType=Rotation, targetBone {boneName=pelvis}}]
    constraintWeights [1,1]
   }
   n78 = "Transform (Modify) Bone - Bone: dyn_skirt_bR_01"(ComponentPose=@n77, Alpha=1, Translation="10.945,0.636,7.988", Rotation="-39.197,-5.844,9.140", Scale="1,1,1") {
    rotationMode=BMM_Additive
    rotationSpace=BCS_BoneSpace
    translationMode=BMM_Additive
    translationSpace=BCS_BoneSpace
    alphaBoolBlend {blendOption=Linear}
    boneToModify {boneName=dyn_skirt_bR_01}
    rotation=<rot -39.197 -5.844 9.14>
    translation=<10.945 0.636 7.988>
   }
   n79 = "Transform (Modify) Bone - Bone: dyn_roll_02"(ComponentPose=@n78, Alpha=1, Translation="0,0,0", Rotation="-50.000,0.078,-0.125", Scale="1,1,1") {
    rotationMode=BMM_Additive
    rotationSpace=BCS_BoneSpace
    translationMode=BMM_Additive
    translationSpace=BCS_BoneSpace
    alphaBoolBlend {blendOption=Linear}
    boneToModify {boneName=dyn_roll_02}
    rotation=<rot -50 0.078 -0.125>
   }
   Output Pose(Result=Blend Poses by bool(
     BlendPose_0=Component To Local(ComponentPose=@n79)
     BlendPose_1=Blend Poses by bool(BlendPose_0=@n42, BlendPose_1=@n56, bActiveValue=@n57) {blendType=Linear}
     bActiveValue=Get VehicleBool() -> (VehicleBool:bool)
    ) {blendType=Linear}) {layerGroup=None}
   n80 = Control Rig(Source=Copy Pose From Mesh() {bUseAttachedParent=true}) {
    alphaBoolBlend {blendOption=Linear}
    controlRigAssetReference {blueprintRigClass=&/F_MED_DeformRig_NitroFlow_CtrlRig_.F_MED_DeformRig_NitroFlow_CtrlRig__C}
   }
   n81 = Control Rig!(Source=@n80, Alpha, IsCrouchCtrl:bool=Get bIsCrouching(), IsSlidingCtrl:bool=Get bIsSliding()) -> (Pose) {
    alphaBoolBlend {blendOption=Linear}
    controlRigAssetReference {blueprintRigClass=&/F_MED_NitroFlow_Royal_CtrlRig.F_MED_NitroFlow_Royal_CtrlRig_C}
   }
   Save cached pose 'beforeRban'(Pose=@n81) {cacheName=beforeRban}
   Get in vehicle scale alpha()
   Get in vehicle scale alpha()
   Get bIsInVehicle() -> (bIsInVehicle:bool)
  ]}
 EventGraph {nodes [
   Event BlueprintUpdateAnimation()
   > n1 = Sequence!(execute) -> (then_0, then_1, then_2:Exec, then_3:Exec)
   > n2 = Gravity Override State Function DEPRECATED(CharacterPartAnimInstance=Self-Reference(), GameTimeInSeconds=GetGameTimeInSeconds(), GravityOverrideParams=Get Gravity Override Params())
   > Set Gravity Override(Gravity Override=@n2)
   n3 = Get bIsTacticalSprinting()
   n4 = Get bIsInFrontEnd()
   > Set NormalRbanAlpha(NormalRbanAlpha:double=SelectFloat(A=0, B=1, bPickA=OR Boolean(A=@n4, B=@n3))) -> (Output_Get:double)
   n5 = OR Boolean(A=Get bIsInFrontEnd(), B=Get bIsPlayingEmote())
   > Set FrontEndIdle(FrontEndIdle:bool=@n5) -> (Output_Get:bool)
   > Set DynamicAlpha(DynamicAlpha:double=SelectFloat(A=0, B=1, bPickA=@n5)) -> (Output_Get:double)
   n6 = Get bIsCrouchSprinting()
   > Set SkirtScale(SkirtScale:bool=AND Boolean(A=Get bIsCrouching(), B="Not Equal (Boolean)"(A=@n6, B=true) -> (ReturnValue=~))) -> (Output_Get:bool)
   n7 = Branch(execute=@n1.then_1, Condition=Get bIsInFrontEnd())
   > Set BwdMovement(BwdMovement:double=0) -> (Output_Get:double)
   n8 = PlayerMovement(execute=@n7.else, AnimInstance=Self-Reference(), XDotProductCoefficient=1, YDotProductCoefficient=1, ZDotProductCoefficient=1)
   n9 = float + float(A=float * float(A="Absolute (Float)"(A=@n8), B=0.5), B=MapRangeClamped(Value=@n8.YMovement, InRangeA=0, InRangeB=-1)) {is=title K2Node_CommutativeAssociativeBinaryOperator}
   n10 = MapRangeClamped(Value=@n9, InRangeA=0, InRangeB=1.1, OutRangeA=0.8, OutRangeB=0)
   > Set BwdMovement(BwdMovement=@n10)
   > Set FwdMovement(FwdMovement:double=MapRangeClamped(Value=@n9, InRangeA=0, InRangeB=1.1)) -> (Output_Get:double)
   > n11 = Branch(Condition=Get bIsCrouching())
   n12 = "Clamp (Float)"(Value=@n8.YMovement, Min=0, Max=1)
   > Set CrouchScale(CrouchScale:double="Clamp (Float)"(Value=float - float(A=@n10, B=@n12), Min=0, Max=1)) -> (Output_Get:double)
   Set CrouchScale(execute=@n11.else, CrouchScale=0)
   n13 = Branch(execute=@n1.then_2, Condition=Get bIsSkyDiving())
   > n14 = Branch(Condition="float > float"(A=Get PawnVelocityZ(), B=50) {is=title K2Node_CallFunction})
   > Set BwdMovement(BwdMovement=float * float(A=Get BwdMovement(), B=0))
   Set BwdMovement(execute=[@n14.else, @n13.else], BwdMovement=float * float(A=Get BwdMovement(), B=1))
   n15 = Vehicle Dynamics Scaler Function!(execute=@n1.then_3, CharacterPartAnimInstance:Custom Character Part Anim Instance Object Reference=Self-Reference(), Debug:bool=false) -> (then, InVehicleHeadScaleAlpha:double, InVehicleHeadSimAlpha:double, InVehicleSkirtScaleAlpha:double, InVehicleSkirtSimAlpha:double)
   > n16 = Set in vehicle scale alpha(in vehicle scale alpha:double=@n15.InVehicleSkirtScaleAlpha) -> (Output_Get:double)
   > Set in vehicle rigid alpha(in vehicle rigid alpha:double=@n15.InVehicleSkirtSimAlpha) -> (Output_Get:double)
   > Set VehicleBool(VehicleBool:bool="Equal (Float)"(A=@n16, B=1)) -> (Output_Get:bool)
   Branch(Condition="Equal (Name)"!(A:Name=Event OnNotifyPartsStopped!() -> (OutputDelegate, then, Context:Name=None) {is=title K2Node_Event}.Context, B:Name=NitroFlow_Idle) -> (ReturnValue:bool) {is=title K2Node_CallFunction})
   > Set FrontEnd_Idle_On(FrontEnd_Idle_On:bool=false) -> (Output_Get:bool)
   Branch(Condition="Equal (Name)"(A=Event OnNotifyPartsStarted!() -> (OutputDelegate, then, Context:Name=None) {is=title K2Node_Event}.Context, B=NitroFlowFrontIdle))
   > Set FrontEnd_Idle_On(FrontEnd_Idle_On=true)
   Get bIsInVehicle()
  ]}
]
variables [
 Alpha {type=number, default=1}
 BwdMovement {type=number}
 CrouchScale {type=number}
 DynamicAlpha {type=number}
 FrontEnd_Idle_On {type=boolean}
 FrontEndIdle {type=boolean}
 FwdMovement {type=number}
 Gravity Override {type=Vector}
 Gravity Override Params {
  type=GravityOverrideParamsStruct
  default {
   skydive_FloatingDown {gravityOverride=<0 -500 200>}
   skydive_GoingUp {gravityOverride=<0 -300 -300>}
   tactical_Sprint {gravityOverride=<0 -400 -1000>}
  }
 }
 in vehicle rigid alpha {type=number}
 in vehicle scale alpha {type=number}
 NormalRbanAlpha {type=number}
 SkirtScale {type=boolean}
 VehicleBool {type=boolean}
]
